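# Additive glow material for the SpikeMiner Wireframe vehicle wheel, tinted by a PaintColor parameter.
# A static, stretched noise gradient masks a high-contrast noise texture panning along V, over a faint base glow from the same gradient.
# The clamped result, scaled by EmissiveIntensity and the paint color, drives emissive.
# Material attributes then pass through the shared vehicle damage and de-rez transition functions.
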Material /VehicleCosmetics/Wheels/SpikeMiner_Wireframe/Materials/M_Wheel_SpikeMinerWireframe_Glow
graph {
 properties {
  bAutomaticallySetUsageInEditor=false
  bUseMaterialAttributes=true
  bUsedWithSkeletalMesh=true
  blendMode=BLEND_Additive
  usedShadingModels=Default Lit
 }
 nodes [
  n1 = TextureCoordinate() {uTiling=0.1, vTiling=-1}
  n2 = Constant2Vector() {g=0.99}
  n3 = ConstantBiasScale(None=TextureSample(UVs=Add(A=@n1, B=@n2)) {texture=/VehicleCosmetics/Wheels/_SharedTextures/FX/Universal_Noise_34_RGB}.R) {bias=-0.001, scale=2}
  n4 = TextureCoordinate() {uTiling=10, vTiling=3}
  n5 = ConstantBiasScale(None=TextureSample(UVs=Panner(Coordinate=@n4) {speedY=0.5}) {samplerType=SAMPLERTYPE_LinearColor, texture=/VehicleCosmetics/Wheels/Trix_Radiant/Textures/Universal_Noise_Heightmap_02}.R) {bias=-0.09, scale=100}
  n6 = Add(A=Multiply(A=@n3, B=Constant() {r=0.2}), B=Multiply(A=@n3, B=Clamp(None=@n5) {maxDefault=10}))
  n7 = Multiply(
   A=Multiply(A=Clamp(None=@n6), B=ScalarParameter() {defaultValue=2, parameterName=EmissiveIntensity})
   B=VectorParameter() {parameterName=PaintColor, defaultValue=<rgba 0.2664 0.2664 0.2664 1>}.RGB
  )
  MaterialFunctionCall(In=MaterialFunctionCall(Material=SetMaterialAttributes!(MaterialAttributes, Emissive Color=@n7, Opacity) {attributeSetTypes [#1, #2]}) {materialFunction=&MF_BaseVehicle_Damage}.Result) {materialFunction=&MF_VehicleDeRez_Transition}
 ]
}
